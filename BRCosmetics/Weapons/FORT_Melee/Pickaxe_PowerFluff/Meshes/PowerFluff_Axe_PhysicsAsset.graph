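# Physics asset for the PowerFluff pickaxe cosmetic, giving its dyn_ bones capsule bodies and joints for secondary motion.
# A spine chain runs from dyn_Spine_MSTR through Spine_5, then branches into a head and two-segment left and right hands.
# The first joint swings and twists freely, while the rest have tight swing limits and almost no twist.
# An unconstrained Char_Thigh_r capsule sits outside the chain, apparently as a character-leg collider.
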PhysicsAsset /BRCosmetics/Weapons/FORT_Melee/Pickaxe_PowerFluff/Meshes/PowerFluff_Axe_PhysicsAsset
bodies {
 Char_Thigh_r [{shapeType=Capsule, length=47.5376, radius=11.4624, shapeName=Char_Thigh_r_capsule, center=<9.0991 0.2988 0>, rotation=<rot -90 0 0>}]
 dyn_Hand_l_01 [{shapeType=Capsule, length=1.3058, radius=1.5642, shapeName=dyn_Hand_l_01_capsule, center=<1.7 0 0.3>, rotation=<rot 90 0 0>}]
 dyn_Hand_l_02 [{
   shapeType=Capsule
   length=4.5736
   radius=2.3451
   shapeName=dyn_Hand_l_02_capsule
   center=<4.5 -0.25 0.0153>
   rotation=<rot 84.3049 -0.008 0.0004>
  }]
 dyn_Hand_r_01 [{shapeType=Capsule, length=1.354, radius=1.516, shapeName=dyn_Hand_r_01_capsule, center=<1.7 0 -0.3>, rotation=<rot 90 0 0>}]
 dyn_Hand_r_02 [{
   shapeType=Capsule
   length=3.6458
   radius=2.4677
   shapeName=dyn_Hand_r_02_capsule
   center=<4.5 -0.3425 0.0153>
   rotation=<rot -84 0.0001 -6.2597>
  }]
 dyn_Head_01 [{shapeType=Capsule, length=0, radius=14.3388, shapeName=dyn_Head_01_capsule, center=<12.5056 0 0>, rotation=<rot 0 0 -90>}]
 dyn_Spine_0 [{shapeType=Capsule, length=4, radius=4, shapeName=dyn_Spine_0_capsule, center=<7 0 0>}]
 dyn_Spine_1 [{shapeType=Capsule, length=1.3295, radius=4, shapeName=dyn_Spine_1_capsule, center=<7 0 0>}]
 dyn_Spine_2 [{shapeType=Capsule, length=4.4184, radius=4.3112, shapeName=dyn_Spine_2_capsule, center=<7 0 0.3327>, rotation=<rot 90 0 0>}]
 dyn_Spine_3 [{shapeType=Capsule, length=1.8494, radius=2.7534, shapeName=dyn_Spine_3_capsule, center=<7 0 0>, rotation=<rot 90 0.3427 1.0634>}]
 dyn_Spine_4 [{shapeType=Capsule, length=1, radius=3.8183, shapeName=dyn_Spine_4_capsule, center=<7 0 0>, rotation=<rot 90 0 0>}]
 dyn_Spine_5 [{shapeType=Capsule, length=1, radius=5, shapeName=dyn_Spine_5_capsule, center=<7 0 0>, rotation=<rot 90 0 0>}]
 dyn_Spine_MSTR [{
   shapeType=Capsule
   length=3.8804
   radius=2.7312
   shapeName=dyn_Spine_MSTR_capsule
   center=<-0.2455 0.1355 -3.6256>
   rotation=<rot 0.1179 -90 3.1178>
  }]
}
constraints [
 {
  bone1Name=dyn_Spine_0
  bone2Name=dyn_Spine_MSTR
  swing1LimitDegrees=45
  swing1Motion=Free
  swing2LimitDegrees=45
  swing2Motion=Free
  twistLimitDegrees=45
  twistMotion=Free
 }
 {bone1Name=dyn_Spine_1, bone2Name=dyn_Spine_0, swing1LimitDegrees=25, swing2LimitDegrees=0, twistLimitDegrees=0}
 {bone1Name=dyn_Spine_2, bone2Name=dyn_Spine_1, swing1LimitDegrees=15, swing2LimitDegrees=5, twistLimitDegrees=1}
 {bone1Name=dyn_Spine_3, bone2Name=dyn_Spine_2, swing1LimitDegrees=15, swing2LimitDegrees=5, twistLimitDegrees=0}
 {bone1Name=dyn_Spine_4, bone2Name=dyn_Spine_3, swing1LimitDegrees=15, swing2LimitDegrees=5, twistLimitDegrees=0}
 {bone1Name=dyn_Spine_5, bone2Name=dyn_Spine_4, swing1LimitDegrees=15, swing2LimitDegrees=5, twistLimitDegrees=0}
 {bone1Name=dyn_Hand_l_01, bone2Name=dyn_Spine_5, swing1LimitDegrees=15, swing2LimitDegrees=25, twistLimitDegrees=1}
 {bone1Name=dyn_Hand_l_02, bone2Name=dyn_Hand_l_01, swing1LimitDegrees=15, swing2LimitDegrees=30, twistLimitDegrees=1}
 {bone1Name=dyn_Hand_r_01, bone2Name=dyn_Spine_5, swing1LimitDegrees=15, swing2LimitDegrees=25, twistLimitDegrees=1}
 {bone1Name=dyn_Hand_r_02, bone2Name=dyn_Hand_r_01, swing1LimitDegrees=15, swing2LimitDegrees=30, twistLimitDegrees=1}
 {bone1Name=dyn_Head_01, bone2Name=dyn_Spine_5, swing1LimitDegrees=15, swing2LimitDegrees=5, twistLimitDegrees=0}
]
